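# Additive unlit material that fakes a light beam on a line billboard, tinted by time of day.
# MF_LineBillboard builds the WPO shape from length, base width and beam angle, with a rounded tip.
# UVs derived from local position avoid taper distortion and drive emissive edge diffusion, distance attenuation and tip fade.
# Depth, near-camera and end-on view fades form opacity, and a mobile shading-path switch adjusts brightness.
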Material /WildEstate/Environment/Props/LightSource/BR_LightCone/M_LightBeamBillboard
graph {
 properties {bAutomaticallySetUsageInEditor=false, blendMode=BLEND_Additive, shadingModel=MSM_Unlit, usedShadingModels=Unlit}
 nodes [
  n1 = LocalPosition()
  ComponentMask(None=@n1.XYZ) {b=true}
  Constant3Vector() {constant=<rgba 1 1 1 0>}
  n2 = NamedRerouteUsage()
  n3 = Multiply(A=@n2, B=NamedRerouteUsage())
  MaterialFunctionCall(Emissive=Multiply(A=ShadingPathSwitch(Default=@n2, Mobile=@n3), B=NamedRerouteUsage())) {materialFunction=&MF_FX_EmissiveAdjustments}
  n4 = ScalarParameter() {
   defaultValue=0.07
   desc=Set the beam end tip roundness
   group=Light Beam Settings
   parameterName=Tip Roundness
   sliderMax=0.25
   sortPriority=5
  }
  NamedRerouteDeclaration(None=@n4) {name=Roundness, nodeColor=<rgba 0 1 0.0648 1>}
  n5 = TextureCoordinate()
  n6 = Multiply(A=Distance(A=ComponentMask(None=@n5) {r=true}, B=Constant() {r=0.5})) {constB=2}
  n7 = Multiply(A=OneMinus(None=Power(Base=@n6)), B=OneMinus(None=ComponentMask(None=@n5) {g=true}))
  NamedRerouteDeclaration(None=LinearInterpolate(B=Saturate(None=@n7), Alpha=NamedRerouteUsage()) {constA=1}) {name=RoundedOffset, nodeColor=<rgba 1 0.8587 0.0832 1>}
  n8 = ScalarParameter() {defaultValue=700, desc="Near fade, use to adjust camera clipping", group=Light Beam Camera Fade, parameterName=NearFade, sortPriority=31}
  NamedRerouteDeclaration(None=@n8) {name=NearFade, nodeColor=<rgba 1 0 0.0703 1>}
  n9 = Divide(A=ComponentMask(None=@n1.XYZ) {r=true}, B=NamedRerouteUsage()) {constB=1}
  n10 = Divide(
   A=Length(None=ComponentMask(None=@n1.XYZ) {b=true, g=true}) {is=class MaterialExpressionLength}
   B=LinearInterpolate(A=NamedRerouteUsage(), B=NamedRerouteUsage(), Alpha=Saturate(None=@n9))
  ) {constB=0.0025}
  NamedRerouteDeclaration(None=AppendVector(A=Add(A=@n10) {constB=0.5}, B=OneMinus(None=@n9))) {name=SmoothUVs, nodeColor=<rgba 0.6641 1 0 1>}
  n11 = ScalarParameter() {
   defaultValue=1
   desc=Default HDRMult is too bright for mobile. This allows to reduce for mobile when HDR is high.
   group=Light Beam Settings
   parameterName=MobileBrightnessMultiplier
  }
  NamedRerouteDeclaration(None=@n11) {name=MobileBrightnessMultiplier, nodeColor=<rgba 0.7109 1 0 1>}
  n12 = Multiply(A=SmoothStep(Value=NamedRerouteUsage()) {constMax=0.1, constValue=1}, B=NamedRerouteUsage()) {constA=1}
  n13 = Multiply(A=NamedRerouteUsage(), B=@n12)
  n14 = MaterialFunctionCall() {materialFunction=&ObjectScale}
  n15 = NamedRerouteUsage()
  n16 = Tangent!(None=Multiply(A=NamedRerouteUsage(), B=Divide() {constB=360}))
  n17 = Add(A=@n15, B=Multiply(A=@n12, B=Multiply(A=@n16) {constB=2}))
  NamedRerouteDeclaration(None=MaterialFunctionCall!(
    Camera Relative World Position
    TangentU_WS
    TangentV_WS
    UVs for Thickness
    UVs for Projection
    Input Expand U or V UV Channel
    Lenght=Multiply(A=@n14.Scale X, B=@n13)
    WidthBase=Multiply(A=@n14.Scale X, B=@n15)
    WidthTip=Multiply(A=@n14.Scale X, B=@n17)
    UVs For NormalShading
   ) {materialFunction=&/MF_LineBillboard}.WPO) {name=Final WPO, nodeColor=<rgba 1 0.9399 0 1>}
  n18 = ScalarParameter() {defaultValue=500, desc=Lenght of the light beam, group=Light Beam Settings, parameterName=Length, sliderMax=8000, sortPriority=0}
  NamedRerouteDeclaration(None=@n18) {name=BeamLength, nodeColor=<rgba 0 1 0.0648 1>}
  n19 = VectorParameter() {group=Light Beam Settings, parameterName=Color Day, sortPriority=-1, defaultValue=<rgba 0 0 0 1>}
  n20 = VectorParameter() {group=Light Beam Settings, parameterName=Color Night, sortPriority=-1, defaultValue=<rgba 1 0.9937 0.8113 1>}
  NamedRerouteDeclaration(None=MaterialFunctionCall(Color_Morning=@n19.RGB, Color_Day=@n19.RGB, Color_Evening=@n20.RGB, Color_Night=@n20.RGB) {materialFunction=&Color_Time_of_Day}.Result) {name=Color, nodeColor=<rgba 1 0.0469 0 1>}
  n21 = NamedRerouteUsage()
  n22 = Multiply(A=Distance(A=ComponentMask(None=@n21) {r=true}, B=Constant() {r=0.5})) {constB=2}
  n23 = Saturate(None=SmoothStep(Max=NamedRerouteUsage(), Value=OneMinus(None=@n22)) {constMax=1.5})
  n24 = NamedRerouteUsage()
  n25 = ComponentMask(None=@n21) {g=true}
  n26 = OneMinus(None=@n25)
  n27 = Multiply(A=Power(Base=Multiply(A=@n26) {constB=10}), B=NamedRerouteUsage())
  n28 = Multiply(A=Saturate(None=SmoothStep(Max=@n24, Value=@n25)), B=Divide(B=Add(A=@n27) {constA=1}))
  NamedRerouteDeclaration(None=Multiply(A=Multiply(A=@n23, B=@n28), B=NamedRerouteUsage()) {constB=10}) {name=Emissive, nodeColor=<rgba 0 0.6511 1 1>}
  n29 = DepthFade()
  n30 = QualitySwitch(Default=@n29, High=@n29, Medium=@n29)
  n31 = MaterialFunctionCall(Distance=NamedRerouteUsage()) {materialFunction=&Near_Camera_Fade}
  n32 = CameraVectorWS()
  n33 = Normalize(VectorInput=Transform(None=Constant3Vector() {constant=<rgba 1 0 0 0>}) {transformSourceType=TRANSFORMSOURCE_Local})
  n34 = SmoothStep(Min=NamedRerouteUsage(), Max=NamedRerouteUsage(), Value=Abs(None=DotProduct(A=@n32, B=@n33))) {constMax=0.9, constMin=0.6}
  NamedRerouteDeclaration(None=Multiply(A=Multiply(A=@n30, B=@n31.Result), B=Saturate(None=@n34))) {name=Opacity, nodeColor=<rgba 1 0 0.6185 1>}
  n35 = ScalarParameter() {
   defaultValue=0.9
   desc=Adjust view angle where beam starts to gradually fade when front/back faced directly.
   parameterName=ViewAlignFadeStart
   sliderMax=1
  }
  NamedRerouteDeclaration(None=@n35) {name=ViewAngleFadeStart, nodeColor=<rgba 0 1 0.7422 1>}
  n36 = ScalarParameter() {
   defaultValue=1
   desc=Adjust view angle where beam fades completly when front/back faced directly.
   parameterName=ViewAlignFadeEnd
   sliderMax=1
  }
  NamedRerouteDeclaration(None=@n36) {name=ViewAngleFadeEnd, nodeColor=<rgba 1 0.4219 0 1>}
  n37 = ScalarParameter() {
   defaultValue=0.5
   desc=How diffuse the beam gets with distance
   group=Light Beam Settings
   parameterName=Diffusion
   sliderMax=1
   sortPriority=4
  }
  NamedRerouteDeclaration(None=@n37) {name=Diffusion, nodeColor=<rgba 1 0 0.5391 1>}
  n38 = ScalarParameter() {
   defaultValue=1
   desc="Light attenuation coefficient. Lower values makes attenuation more linear.\r\nAttenuation takes a long distance to fall close to zero, so use TipFade property to adjust the beam termination"
   group=Light Beam Settings
   parameterName=Attenuation
   sliderMax=5
   sortPriority=6
  }
  NamedRerouteDeclaration(None=@n38) {name=Attenuation, nodeColor=<rgba 0.8984 0 1 1>}
  NamedRerouteDeclaration(None=ScalarParameter() {defaultValue=30, desc=Light beam base width, group=Light Beam Settings, parameterName=Beam Width, sortPriority=1}) {name=Beam Width, nodeColor=<rgba 0 1 0.0648 1>}
  n39 = ScalarParameter() {defaultValue=15, desc=How wide the beam projects, group=Light Beam Settings, parameterName=BeamAngle, sliderMax=60, sortPriority=2}
  NamedRerouteDeclaration(None=@n39) {name=Beam Angle, nodeColor=<rgba 0 0.7891 1 1>}
  n40 = ScalarParameter() {
   defaultValue=0.2
   desc="Sets how smoothly the beam disappears at the end tip.\r\nManly used to avoid a harsh line where the beam ends."
   group=Light Beam Settings
   parameterName=TipFade
   sliderMax=1
   sortPriority=7
  }
  NamedRerouteDeclaration(None=@n40) {name=Tip Fade, nodeColor=<rgba 0 1 0.8828 1>}
  n41 = ScalarParameter() {defaultValue=1, desc=Brightness intensity, group=Light Beam Settings, parameterName=Intensity, sliderMax=50, sortPriority=3}
  NamedRerouteDeclaration(None=@n41) {name=Intensity, nodeColor=<rgba 0.9047 0 0.6785 1>}
  NamedRerouteDeclaration(None=@n17) {name=Beam Width end, nodeColor=<rgba 0 1 0.7188 1>}
  NamedRerouteUsage()
  NamedRerouteUsage()
  TextureCoordinate()
 ]
}
